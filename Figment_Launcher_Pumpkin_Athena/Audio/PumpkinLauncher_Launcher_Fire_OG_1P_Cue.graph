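# First-person fire sound cue for the Pumpkin Launcher weapon (OG variant).
# A single grenade-launcher fire wave from the shared weapon sounds feeds a Modulator that randomizes pitch from normal up to a higher value.
# The cue boosts volume, uses first-person projectile attenuation and the ranged-fire 1P sound class, and overrides concurrency to block new instances once the limit is reached.
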SoundCue /Figment_Launcher_Pumpkin_Athena/Audio/PumpkinLauncher_Launcher_Fire_OG_1P_Cue
root=Modulator_6
nodes [
 {id=Modulator_6, children [WavePlayer_0], properties {pitchMax=1.5, pitchMin=1}}
 {id=WavePlayer_0, properties {soundWaveAssetPtr=/Game/Sounds/Fort_Weapons/Launchers/weapon_grenade_fire_04}}
]
properties {
 attenuationSettings=&Fort_Weapon_Projectile_FirstPerson_Athena_Att
 bOverrideConcurrency=true
 maxDistance=2097152
 soundClassObject=&Fort_WeaponRanged_Fire_1P_PSM
 volumeMultiplier=1.4441
 attenuationOverrides {falloffDistance=5000, attenuationShapeExtents {x=1000}}
 concurrencyOverrides {resolutionRule=PreventNew}
}
