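# Third-person fire sound cue for the Pumpkin Launcher (OG variant), built from existing grenade and rocket launcher waves.
# A mixer layers a pitch-randomized grenade shot with shotgun attenuation over a distance crossfade from a medium rocket shot to a modulated distant one.
# The two mixer branches use separate third-person and distant weapon-fire sound classes, and each wave path has its own attenuation.
# A concurrency override prevents new instances once the limit is reached.
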SoundCue /Figment_Launcher_Pumpkin_Athena/Audio/PumpkinLauncher_Launcher_Fire_OG_3P_Cue
root=Mixer_3
nodes [
 {
  id=Attenuation_10
  children [WavePlayer_42]
  properties {attenuationSettings=&Fort_Weapon_Projectile_Medium_Mono_Athena_Att}
 }
 {
  id=Attenuation_11
  children [Modulator_7]
  properties {attenuationSettings=&Fort_Weapon_Projectile_Distant_Mono_Athena_Att}
 }
 {id=Attenuation_9, children [Modulator_6], properties {attenuationSettings=/Game/Sounds/Attenuation/Fort_Weapon_Shotgun_Sml_Att}}
 {
  id=DistanceCrossFade_3
  children [Attenuation_10, Attenuation_11]
  properties {crossFadeInput [
    {fadeInDistanceEnd=100, fadeInDistanceStart=0, fadeOutDistanceEnd=3000, fadeOutDistanceStart=2000, volume=1}
    {fadeInDistanceEnd=3000, fadeInDistanceStart=1000, fadeOutDistanceEnd=12000, fadeOutDistanceStart=10000, volume=1}
   ]}
 }
 {id=Mixer_3, children [SoundClass_7,SoundClass_6], properties {inputVolume [1,1]}}
 {id=Modulator_6, children [WavePlayer_0], properties {pitchMax=1.5, pitchMin=1}}
 {id=Modulator_7, children [WavePlayer_45], properties {pitchMax=1.1}}
 {
  id=SoundClass_6
  children [DistanceCrossFade_3]
  properties {soundClassOverride=&Fort_WeaponRanged_Fire_Distant}
 }
 {
  id=SoundClass_7
  children [Attenuation_9]
  properties {soundClassOverride=&Fort_WeaponRanged_Fire_3P_PSM}
 }
 {id=WavePlayer_0, properties {soundWaveAssetPtr=/Game/Sounds/Fort_Weapons/Launchers/weapon_grenade_fire_04}}
 {id=WavePlayer_42, properties {soundWaveAssetPtr=/Game/Sounds/Fort_Weapons/Launchers/Rocket_Launcher_Medium_ShotC01}}
 {id=WavePlayer_45, properties {soundWaveAssetPtr=/Game/Sounds/Fort_Weapons/Launchers/Rocket_Launcher_Distant_ShotE01}}
]
properties {
 bOverrideConcurrency=true
 maxDistance=27500
 volumeMultiplier=1.575
 attenuationOverrides {falloffDistance=5000, attenuationShapeExtents {x=1000}}
 concurrencyOverrides {resolutionRule=PreventNew}
}
