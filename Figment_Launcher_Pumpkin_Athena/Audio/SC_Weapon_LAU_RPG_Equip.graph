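# Sound cue for the equip foley of the Pumpkin RPG launcher weapon.
# It reuses a stock launcher reload-end wave, passed through a modulator and a short fixed delay before it plays.
# It uses the weapon foley sound class and the shared weapon reload attenuation.
# A concurrency override allows one instance at a time, stopping the farthest and then preventing new ones.
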SoundCue /Figment_Launcher_Pumpkin_Athena/Audio/SC_Weapon_LAU_RPG_Equip
root=Delay_0
nodes [
 {id=Delay_0, children [Modulator_1], properties {delayMax=0.1, delayMin=0.1}}
 {id=Modulator_1, children [WavePlayer_4]}
 {id=WavePlayer_4, properties {soundWaveAssetPtr=/Game/Sounds/Fort_Weapons/Launchers/weapons_launchers_reload_end_01}}
]
properties {
 attenuationSettings=&Fort_Weapon_Reload_Att
 bOverrideConcurrency=true
 maxDistance=2250
 soundClassObject=&Fort_WeaponFoley
 volumeMultiplier=1.002
 attenuationOverrides {falloffDistance=5000, attenuationShapeExtents {x=1000}}
 concurrencyOverrides {maxCount=1, resolutionRule=StopFarthestThenPreventNew}
}
